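; Large hourglass quest actor for the Season 29 Snapshot quest (Part 4, Q08), extending a parent interactable with falling-sand FX.
; After interaction, it deactivates and hides the sand where cosmetic events can run, only for a locally controlled instigator.
; That check fails on the server, matches split-screen players against local controllers, and otherwise accepts any valid controller.
; The construction script forces the post-interaction mesh to unit world scale.
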(event EventBeginPlay
  (|CallParentFunction) ; This node is disabled and will not be called.\nDrag off pins to build functionality.
)

(event Collision|EventActorBeginOverlap
  (params (OtherActor :type Actor%20Object%20Reference)))

(event EventTick
  (params (DeltaSeconds :type Float%20%28single-precision%29)))

(event VisualState|Meshes|EventPostInteractionVisuals
  (params (PlayerController :type Fort%20Player%20Controller%20Object%20Reference))
  (bind _fallingsand (Variables|Default|GetFallingSand))
  ;[0 0 0] Disable sand FX when interacted with
  (|CallParentFunction PlayerController)
  (Networking|CanExecuteCosmeticEvents
    (:True
      (bind _output (CallFunction|IsInstigatorLocallyControlled PlayerController))
      (if _output
        (Components|Activation|Deactivate _fallingsand)
        (Rendering|SetVisibility _fallingsand false true)))))

(fn IsInstigatorLocallyControlled
  (params (InstigatingPlayerController :type Fort%20Player%20Controller%20Object%20Reference))
  ;[0 0 0] Returns true on the executing client if the instigator controller is locally controlled. Returns false if running on the server
  (if (Networking|IsServer)
    (return false)
    (else
      (Utilities|IsValid InstigatingPlayerController
        (:Is%20Valid
          (if (PlayerController|IsinSplitscreen InstigatingPlayerController)
            (for _array_element (Fortnite|GetLocalFortPlayerControllers true)
              (if (== _array_element InstigatingPlayerController)
                (return true)))
            (return false)
            (else
              (return true))))))))

(fn ConstructionScript
  (|CallParentFunction)
  (Transformation|SetWorldScale3D (Variables|Components|GetSM_Quest_PostInteraction) "1.000000,1.000000,1.000000"))
